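# Experimental unlit, alpha-composite material in the volumetric cloud compositing tools that turns a texture into a hard binary mask, effectively a ceiling for 0–1 values.
# It samples the RT texture parameter at the top mip, and an If node per channel maps values above zero to one and all others to zero.
# The thresholded RGB is recombined with MakeFloat3, while the thresholded alpha is inverted with OneMinus.
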Material /Volumetrics/Tools/CloudCompositing/Materials/Experimental/M_BinaryMask_Ceil
graph {
 properties {blendMode=BLEND_AlphaComposite, pixelDepthOffsetMode=PDOM_Legacy, shadingModel=MSM_Unlit, usedShadingModels=Unlit}
 nodes [
  n1 = TextureSampleParameter2D!(UVs, Level, Apply View MipBias) {constMipValue=0, mipValueMode=TMVM_MipLevel, parameterName=RT, texture=/Engine/EngineResources/AICON-Green}
  n2 = Constant()
  n3 = Constant()
  n4 = Constant()
  MaterialFunctionCall(
   X=If(A=@n1.R, "A > B"=Constant() {r=1}, "A == B"=@n2, "A < B"=@n2)
   Y=If(A=@n1.G, "A > B"=Constant() {r=1}, "A == B"=@n3, "A < B"=@n3)
   Z=If(A=@n1.B, "A > B"=Constant() {r=1}, "A == B"=@n4, "A < B"=@n4)
  ) {materialFunction=&MakeFloat3}
  n5 = Constant()
  OneMinus(None=If(A=@n1.A, "A > B"=Constant() {r=1}, "A == B"=@n5, "A < B"=@n5))
 ]
}
