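# Experimental unlit material in the cloud compositing tools, named as a preview for the cloud razor mask.
# It samples the RT texture parameter, which defaults to a black texture, at a fixed mip level 0.
# The sampled RGB feeds a ConstantBiasScale node at default settings, which remaps signed -1 to 1 values into the 0 to 1 range.
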Material /Volumetrics/Tools/CloudCompositing/Materials/Experimental/M_CloudRazorMaskPreview
graph {
 properties {pixelDepthOffsetMode=PDOM_Legacy, shadingModel=MSM_Unlit, usedShadingModels=Unlit}
 nodes [ConstantBiasScale(None=TextureSampleParameter2D!(UVs, Level, Apply View MipBias) {constMipValue=0, mipValueMode=TMVM_MipLevel, parameterName=RT, texture=&Black}.RGB)]
}
