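# Experimental unlit alpha-composite material for the cloud compositing tool that fills near-empty texels of a cloud mask render target from a razor mask.
# It samples CloudMaskRT and RazorMaskRT, passing the latter through a Subtract with only its A input wired.
# Per channel, cloud values above Threshold are kept, while the rest take the subtracted razor value.
# RGB is repacked and alpha is inverted, with negative emissive allowed.
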Material /Volumetrics/Tools/CloudCompositing/Materials/Experimental/M_FillRazorMask
graph {
 properties {
  bAllowNegativeEmissiveColor=true
  blendMode=BLEND_AlphaComposite
  pixelDepthOffsetMode=PDOM_Legacy
  shadingModel=MSM_Unlit
  usedShadingModels=Unlit
 }
 nodes [
  n1 = TextureSampleParameter2D!(UVs, Level, Apply View MipBias) {constMipValue=0, mipValueMode=TMVM_MipLevel, parameterName=CloudMaskRT, texture=/Engine/EngineResources/AICON-Red}
  n2 = ScalarParameter() {defaultValue=0.01, parameterName=Threshold}
  n3 = MaterialFunctionCall(Float4=Subtract(A=TextureSampleParameter2D() {constMipValue=0, mipValueMode=TMVM_MipLevel, parameterName=RazorMaskRT, texture=&AICON-Red}.RGBA)) {materialFunction=&BreakOutFloat4Components}
  MaterialFunctionCall(
   X=If(A=@n1.R, B=@n2, "A > B"=@n1.R, "A == B"=@n3.R, "A < B"=@n3.R)
   Y=If(A=@n1.G, B=@n2, "A > B"=@n1.G, "A == B"=@n3.G, "A < B"=@n3.G)
   Z=If(A=@n1.B, B=@n2, "A > B"=@n1.B, "A == B"=@n3.B, "A < B"=@n3.B)
  ) {materialFunction=&MakeFloat3}
  OneMinus(None=If(A=@n1.A, B=@n2, "A > B"=@n1.A, "A == B"=@n3.A, "A < B"=@n3.A))
 ]
}
