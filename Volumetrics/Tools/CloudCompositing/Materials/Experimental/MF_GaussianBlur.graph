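# Experimental cloud-compositing material function that Gaussian-blurs a texture at a chosen mip level.
# A Custom HLSL node builds a sigma-driven kernel capped at 33 taps, runs a single-pass 2D loop, and normalises by the squared kernel sum.
# A static switch injects a preprocessor define via dummy Custom nodes, optionally weighting samples by mask red, blended toward a step threshold.
# Normalisation ignores mask weights, so masking also attenuates output.
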MaterialFunction /Volumetrics/Tools/CloudCompositing/Materials/Experimental/MF_GaussianBlur
graph {nodes [
  MaterialExpressionFunctionInput_1 = FunctionInput(Preview=TextureObject() {texture=/Engine/VREditor/Devices/Vive/UE_Logo}) {inputName=Tex, inputType=FunctionInput_Texture2D, sortPriority=-10}
  MaterialExpressionFunctionInput_5 = FunctionInput(Preview=TextureCoordinate()) {bUsePreviewValueAsDefault=true, inputName=UV, inputType=FunctionInput_Vector2, sortPriority=-5}
  MaterialExpressionFunctionInput_2 = FunctionInput() {inputName=ResolutionXY, inputType=FunctionInput_Vector2, previewValue=<1024 1024 0 1>}
  MaterialExpressionFunctionInput_4 = FunctionInput() {bUsePreviewValueAsDefault=true, inputName=BellSigma, inputType=FunctionInput_Scalar, sortPriority=4, previewValue=<7 0 0 1>}
  MaterialExpressionFunctionInput_0 = FunctionInput() {bUsePreviewValueAsDefault=true, inputName=MipLevel, inputType=FunctionInput_Scalar, sortPriority=2}
  MaterialExpressionFunctionInput_3 = FunctionInput() {
   bUsePreviewValueAsDefault=true
   description=Max value: 33
   inputName=KernelSize
   inputType=FunctionInput_Scalar
   sortPriority=3
   previewValue=<15 0 0 1>
  }
  MaterialExpressionStaticSwitch_0 = StaticSwitch(
   True=Custom!(None) {code="##define USE_MASK_TEXTURE 1\r\n\r\nreturn 0;", description="##define USE_MASK_TEXTURE 1", outputType=CMOT_Float1}
   False=Custom() {code="##define USE_MASK_TEXTURE 0\r\n\r\nreturn 0;", description="##define USE_MASK_TEXTURE 0", outputType=CMOT_Float1}
   Value=FunctionInput(Preview=StaticBool()) {bUsePreviewValueAsDefault=true, inputName=UseMaskTexture, inputType=FunctionInput_StaticBool, sortPriority=66}
  )
  MaterialExpressionFunctionInput_7 = FunctionInput(Preview=TextureObject() {texture=/Engine/EngineResources/WhiteSquareTexture}) {bUsePreviewValueAsDefault=true, inputName=MaskTex, inputType=FunctionInput_Texture2D, sortPriority=67}
  MaterialExpressionFunctionInput_8 = FunctionInput(Preview=TextureCoordinate()) {bUsePreviewValueAsDefault=true, inputName=MaskUV, inputType=FunctionInput_Vector2, sortPriority=68}
  MaterialExpressionFunctionInput_9 = FunctionInput() {bUsePreviewValueAsDefault=true, inputName=MaskStep, inputType=FunctionInput_Scalar, sortPriority=69, previewValue=<0.5 0 0 1>}
  MaterialExpressionFunctionInput_10 = FunctionInput() {bUsePreviewValueAsDefault=true, inputName=MaskStepAlpha, inputType=FunctionInput_Scalar, sortPriority=70, previewValue=<0.5 0 0 1>}
  n1 = Custom!(
   Tex=@MaterialExpressionFunctionInput_1
   UV=@MaterialExpressionFunctionInput_5
   ResolutionXY=@MaterialExpressionFunctionInput_2
   BellSigma=@MaterialExpressionFunctionInput_4
   Level=@MaterialExpressionFunctionInput_0
   KernelSize=@MaterialExpressionFunctionInput_3
   MacroMagic=@MaterialExpressionStaticSwitch_0
   MaskTex=@MaterialExpressionFunctionInput_7
   MaskUV=@MaterialExpressionFunctionInput_8
   MaskStep=@MaterialExpressionFunctionInput_9
   MaskStepAlpha=@MaterialExpressionFunctionInput_10
  ) {
   code="##define MAX_KERNEL_SIZE 33\r\n##define WRAP_EDGE 0\r\n\r\nstruct Functions\r\n{\r\n  float GaussianFunction(in float x, in float o)\r\n  {\r\n    return 0.39894/o * exp(-x*x/(2*o*o));\r\n  }\r\n};\r\nFunctions fun;\r\n\r\nKernelSize = min(KernelSize, MAX_KERNEL_SIZE);\r\nint kernelReach = (KernelSize-1)/2;\r\nfloat kernel[MAX_KERNEL_SIZE];\r\nfloat4 output = float4(0,0,0,0);\r\n\t\t\r\nfor (int i = 0; i <= kernelReach; ++i)\r\n  {\r\n    kernel[kernelReach+i] = fun.GaussianFunction(i, BellSigma);\r\n    kernel[kernelReach-i] = kernel[kernelReach+i];\r\n  }\r\n\t\t\r\nfor (int i=-kernelReach; i <= kernelReach; ++i)\r\n  {\r\n    for (int j=-kernelReach; j <= kernelReach; ++j)\r\n      {\r\n\tfloat2 sampleUV = UV + float2(i,j) / ResolutionXY;\r\n##if WRAP_EDGE\r\n\tsampleUV = saturate(sampleUV);\r\n##endif\r\n\tfloat4 accumulate =\r\n\t  kernel[kernelReach+j]*kernel[kernelReach+i]\r\n\t  *Tex.SampleLevel(TexSampler, sampleUV, MipLevel).rgba;\r\n\r\n##if USE_MASK_TEXTURE\r\n        float2 MaskSampleUV = MaskUV + float2(i,j) / ResolutionXY;\r\n##if WRAP_EDGE\r\n\tMaskSampleUV = saturate(MaskSampleUV);\r\n##endif\r\n        float4 mask = MaskTex.SampleLevel(MaskTexSampler, MaskSampleUV, MipLevel).rrrr;\r\n        accumulate *= lerp(mask, step(mask, MaskStep), MaskStepAlpha);\r\n##endif\r\n\r\n        output += accumulate;\r\n      }\r\n  }\r\n\r\nfloat kernelSum = 0.0;\r\nfor (int i = 0; i < KernelSize; ++i)\r\n  {\r\n    kernelSum += kernel[i];\r\n  }\r\n\r\nreturn output / (kernelSum*kernelSum);\r\n"
   outputType=CMOT_Float4
   inputs [
    {
     inputName=Tex
     input {expression=&self:MaterialExpressionFunctionInput_1}
    }
    {
     inputName=UV
     input {expression=&self:MaterialExpressionFunctionInput_5}
    }
    {
     inputName=ResolutionXY
     input {expression=&self:MaterialExpressionFunctionInput_2}
    }
    {
     inputName=BellSigma
     input {expression=&self:MaterialExpressionFunctionInput_4}
    }
    {
     inputName=MipLevel
     input {expression=&self:MaterialExpressionFunctionInput_0}
    }
    {
     inputName=KernelSize
     input {expression=&self:MaterialExpressionFunctionInput_3}
    }
    {
     inputName=MacroMagic
     input {expression=&self:MaterialExpressionStaticSwitch_0}
    }
    {
     inputName=MaskTex
     input {expression=&self:MaterialExpressionFunctionInput_7}
    }
    {
     inputName=MaskUV
     input {expression=&self:MaterialExpressionFunctionInput_8}
    }
    {
     inputName=MaskStep
     input {expression=&self:MaterialExpressionFunctionInput_9}
    }
    {
     inputName=MaskStepAlpha
     input {expression=&self:MaterialExpressionFunctionInput_10}
    }
   ]
  }
  FunctionOutput(None=@n1)
 ]}
