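# Unlit alpha-composite material that blurs a texture, named RT, for the experimental cloud compositing tools.
# MF_GaussianBlur does the work, fed by the texture object plus resolution and mip level parameters, with kernel, sigma and mask inputs left unconnected.
# The blurred result is split into RGB and an inverted alpha channel.
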Material /Volumetrics/Tools/CloudCompositing/Materials/Experimental/M_GaussianBlur
graph {
 properties {blendMode=BLEND_AlphaComposite, pixelDepthOffsetMode=PDOM_Legacy, shadingModel=MSM_Unlit, usedShadingModels=Unlit}
 nodes [
  n1 = MaterialFunctionCall!(
   Tex=TextureObjectParameter() {parameterName=RT}
   UV
   ResolutionXY=ScalarParameter() {defaultValue=1024, parameterName=ResolutionXY}
   MipLevel=ScalarParameter() {parameterName=MipLevel}
   KernelSize
   BellSigma
   UseMaskTexture
   MaskTex
   MaskUV
   MaskStep
   MaskStepAlpha
  ) {materialFunction=&/MF_GaussianBlur}
  ComponentMask(None=@n1.Result) {b=true, g=true, r=true}
  OneMinus(None=ComponentMask(None=@n1.Result) {a=true})
 ]
}
